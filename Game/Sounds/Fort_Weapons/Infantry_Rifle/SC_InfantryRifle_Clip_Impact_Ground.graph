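# Sound cue for the infantry rifle's ejected clip hitting the ground.
# A random node picks one of four equally weighted M1 clip-eject impact waves, then a modulator raises it to a fixed higher pitch.
# It plays at reduced volume through the weapon reload attenuation and a low-ammo degraded ranged-weapon sound class, with a concurrency override that blocks new instances at the limit.
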SoundCue /Game/Sounds/Fort_Weapons/Infantry_Rifle/SC_InfantryRifle_Clip_Impact_Ground
root=Modulator_1
nodes [
 {id=Modulator_1, children [Random_0], properties {pitchMax=1.2, pitchMin=1.2}}
 {id=Random_0, children [WavePlayer_1,WavePlayer_2,WavePlayer_4,WavePlayer_5], properties {weights [1,1,1,1]}}
 {id=WavePlayer_1, properties {soundWaveAssetPtr=&/M1_ClipEject_HitGround_01}}
 {id=WavePlayer_2, properties {soundWaveAssetPtr=&/M1_ClipEject_HitGround_03}}
 {id=WavePlayer_4, properties {soundWaveAssetPtr=&/M1_ClipEject_HitGround_04}}
 {id=WavePlayer_5, properties {soundWaveAssetPtr=&/M1_ClipEject_HitGround_05}}
]
properties {
 attenuationSettings=&Fort_Weapon_Reload_Att
 bOverrideConcurrency=true
 maxDistance=2250
 soundClassObject=/Game/Sounds/FortSoundClassesAndModes/Fort_WeaponRanged_LowAmmo_Degraded_PSM
 volumeMultiplier=0.5
 concurrencyOverrides {resolutionRule=PreventNew}
}
